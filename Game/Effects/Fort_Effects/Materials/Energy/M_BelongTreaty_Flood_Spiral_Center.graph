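# Translucent unlit Niagara sprite and mesh particle material that draws a spinning, tinted spiral vortex center.
# Noise-warped smoke and mask textures rotate with a sphere-mask offset added to time, twisting them into spirals for emissive, tinted by particle color and time of day, and opacity.
# Opacity adds a noise-distorted circular edge, optional center masks, depth, view-angle and behind-surface fades.
# Dynamic parameters drive black hole size and rotation speed.
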Material /Game/Effects/Fort_Effects/Materials/Energy/M_BelongTreaty_Flood_Spiral_Center
graph {
 properties {
  bUsedWithNiagaraMeshParticles=true
  bUsedWithNiagaraSprites=true
  blendMode=BLEND_Translucent
  pixelDepthOffsetMode=PDOM_Legacy
  shadingModel=MSM_Unlit
  twoSided=true
  usedShadingModels=Unlit
 }
 nodes [
  n1 = ParticleColor()
  n2 = TextureCoordinate()
  n3 = TextureSample(UVs=TextureCoordinate() {uTiling=0.7, vTiling=0.7}) {samplerType=SAMPLERTYPE_LinearColor, texture=&T_NoisePack_LinearColor}
  n4 = Multiply(A=ComponentMask(None=Subtract(A=@n3.RGB) {constB=0.5}) {g=true, r=true}) {constB=0.1}
  n5 = Multiply(
   A=ScalarParameter() {defaultValue=1, parameterName=Rotator speed}
   B=DynamicParameter() {paramNames [BlackHole size, Rotator Speed, Param3, Param4]}.Rotator Speed
  )
  n6 = SphereMask(A=TextureCoordinate(), B=Constant() {r=0.5}, Radius=Constant() {r=1}, Hardness=Constant()) {hardnessPercent=0}
  n7 = Add(A=Time(), B=Multiply(A=Multiply(A=@n6, B=@n6)) {constB=10})
  n8 = TextureCoordinate()
  n9 = TextureSampleParameter2D!(
   UVs=Rotator(Coordinate=Add(A=@n2, B=@n4), Time=Multiply(A=@n5, B=@n7) {constA=1})
   "DDX(UVs)"=DDX(Value=@n8)
   "DDY(UVs)"=DDY(Value=@n8)
   Apply View MipBias
  ) {mipValueMode=TMVM_Derivative, parameterName=Noise, texture=&T_Inky_Smoke_Tile_Chromatic}
  n10 = TextureCoordinate()
  n11 = OneMinus(None=SphereMask(A=Add(B=@n10), B=Constant() {r=0.5}) {attenuationRadius=1, hardnessPercent=0})
  n12 = Multiply(
   A=Power(Base=@n9.RGB, Exp=LinearInterpolate(Alpha=@n11))
   B=VectorParameter() {parameterName=Base color, defaultValue=<rgba 0.2059 0.0521 2 1>}.RGB
  )
  n13 = ScalarParameter() {defaultValue=0.7, parameterName=Main Bright}
  n14 = Add(B=Add(B=Multiply(A=@n12, B=@n13) {constB=0.7}) {constB=0})
  n15 = Constant() {r=1}
  n16 = VectorParameter() {parameterName=Color Night, defaultValue=<rgba 2 0.6849 1.5726 0>}
  MaterialFunctionCall(Emissive=Multiply(
    A=Multiply(A=@n1.RGB, B=@n14)
    B=MaterialFunctionCall(Color_Morning=@n15, Color_Day=@n15, Color_Evening=@n15, Color_Night=@n16.RGB) {materialFunction=&Color_Time_of_Day}.Result
   )) {materialFunction=&MF_FX_EmissiveAdjustments}
  n17 = MaterialFunctionCall!(Distance Bias, Distance Divisor=ScalarParameter() {defaultValue=50, parameterName=Distance Divisor}, Scene Depth, Pixel Depth) {materialFunction=/Game/Effects/Fort_Effects/Materials/_Functions/MF_DistanceBehindTranslucency_01}
  n18 = Constant() {r=1}
  n19 = TextureCoordinate()
  n20 = TextureSample(UVs=TextureCoordinate() {uTiling=0.7, vTiling=0.7}) {samplerType=SAMPLERTYPE_LinearColor, texture=&T_NoisePack_LinearColor}
  n21 = Multiply(A=ComponentMask(None=Subtract(A=@n20.RGB) {constB=0.5}) {g=true, r=true}) {constB=0.1}
  n22 = SphereMask(A=TextureCoordinate(), B=Constant() {r=0.5}, Radius=Constant() {r=1}, Hardness=Constant()) {hardnessPercent=0}
  n23 = Add(A=Time(), B=Multiply(A=Multiply(A=@n22, B=@n22)) {constB=10})
  n24 = TextureCoordinate()
  n25 = TextureSampleParameter2D(UVs=Rotator(Coordinate=Add(A=@n19, B=@n21), Time=Multiply(A=@n5, B=@n23) {constA=1}), "DDX(UVs)"=DDX(Value=@n24), "DDY(UVs)"=DDY(Value=@n24)) {
   mipValueMode=TMVM_Derivative
   parameterName=Mask_Tex
   texture=/Game/Gadgets/Assets/VinderTech_GliderChute/Glider_Cube_Queen_Female/FX/T_Cheap_blue_space_color
  }
  n26 = TextureCoordinate()
  n27 = OneMinus(None=SphereMask(A=Add(B=@n26), B=Constant() {r=0.5}) {attenuationRadius=1, hardnessPercent=0})
  n28 = Add(A=Power(Base=@n25.RGB, Exp=LinearInterpolate(Alpha=@n27))) {constB=0}
  n29 = Add(B=TextureCoordinate())
  n30 = Constant() {r=0.5}
  n31 = Multiply(
   A=ScalarParameter() {defaultValue=0.5, parameterName=Radius}
   B=DynamicParameter() {paramNames [BlackHole size, Rotator Speed, Param3, Param4]}.BlackHole size
  )
  n32 = Multiply(A=Power(Base=SphereMask(A=@n29, B=@n30, Radius=@n31) {attenuationRadius=0.5, hardnessPercent=0})) {constB=2}
  n33 = LinearInterpolate(B=OneMinus(None=Saturate(None=@n32)), Alpha=ScalarParameter() {parameterName=Use center hole}) {constA=1}
  n34 = SphereMask(A=TextureCoordinate(), B=Constant() {r=0.5}, Radius=Constant() {r=1}, Hardness=Constant())
  n35 = Add(A=Multiply(A=Constant() {r=1}) {constB=0.25}, B=Multiply(A=@n34, B=@n34))
  n36 = Cosine(None=@n35)
  n37 = Subtract(A=Multiply(A=TextureCoordinate())) {constB=0.5}
  n38 = ComponentMask(None=@n37) {r=true}
  n39 = Sine(None=@n35)
  n40 = ComponentMask(None=@n37) {g=true}
  n41 = Subtract(A=Multiply(A=@n36, B=@n38), B=Multiply(A=@n39, B=@n40))
  n42 = Add(A=Multiply(A=@n38, B=@n39), B=Multiply(A=@n40, B=@n36))
  n43 = Multiply(
   A=ScalarParameter() {defaultValue=1, parameterName=Rotator speed}
   B=DynamicParameter() {paramNames [BlackHole size, Rotator Speed, Param3, Param4]}.Rotator Speed
  )
  n44 = Time()
  n45 = Multiply(
   A=Rotator(Coordinate=AppendVector(A=@n41, B=@n42), Time=Multiply(A=@n43, B=@n44)) {speed=0.3}
   B=ComponentMask(None=VectorParameter() {parameterName=UV, defaultValue=<rgba 1 1 1 0>}.RGB) {g=true, r=true}
  )
  n46 = MaterialFunctionCall(Float3=Multiply(A=TextureSampleParameter2D!(UVs=@n45, Apply View MipBias) {parameterName=Alpha, samplerType=SAMPLERTYPE_LinearColor, texture=&T_NoisePack_LinearColor}.RGB) {constB=0.1}) {materialFunction=&BreakOutFloat3Components}
  n47 = Subtract(A=Add(A=TextureCoordinate(), B=@n46.R), B=@n46.G)
  n48 = Constant() {r=0.5}
  n49 = Constant() {r=0.5}
  n50 = Constant()
  n51 = Power(
   Base=Multiply(A=Multiply(A=@n28, B=@n33), B=SphereMask(A=@n47, B=@n48, Radius=@n49, Hardness=@n50))
   Exp=ScalarParameter() {defaultValue=1, parameterName=Exp}
  ) {constExponent=1}
  n52 = TextureCoordinate()
  n53 = Constant() {r=0.5}
  n54 = Power(
   Base=OneMinus(None=SphereMask(A=@n52, B=@n53) {attenuationRadius=0.5, hardnessPercent=0})
   Exp=ScalarParameter() {defaultValue=2, parameterName=Center hole mask}
  )
  n55 = LinearInterpolate(B=Saturate(None=Multiply(A=@n54) {constB=10}), Alpha=ScalarParameter() {parameterName=Use Center mask}) {constA=1}
  n56 = Multiply(
   A=Multiply(A=@n1.A, B=Multiply(A=@n51, B=@n55))
   B=LinearInterpolate(
    A=Constant() {r=1}
    B=ShadingPathSwitch(Default=@n17.Normalized Inverted, Deferred=@n17.Normalized Inverted, Forward=@n18, Mobile=@n18)
    Alpha=ScalarParameter() {defaultValue=1, parameterName=Use ParticleDecal}
   )
  )
  n57 = Multiply(A=@n56)
  n58 = Multiply(A=QualitySwitch(Default=@n56, Low=@n57, High=@n56, Medium=@n57, Epic=@n56)) {constB=1.5}
  n59 = ScalarParameter() {defaultValue=10, parameterName=FadeDistance}
  n60 = ScalarParameter() {parameterName="FadeSensitivity (optional) (S)"}
  n61 = Multiply(A=SphereMask(A=TextureCoordinate(), B=Constant() {r=0.5}) {attenuationRadius=0.5, hardnessPercent=0}) {constB=5}
  Multiply(
   A=Multiply(
    A=DepthFade(Opacity=@n58, FadeDistance=@n59) {fadeDistanceDefault=10}
    B=MaterialFunctionCall("FadeSensitivity (optional)"=@n60) {materialFunction=&FadeBasedOnViewAngle}.Result
   )
   B=LinearInterpolate(B=Saturate(None=@n61), Alpha=ScalarParameter() {parameterName=Use sphere edge mask}) {constA=1}
  )
 ]
}
