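# Unlit, two-sided translucent material for Niagara mesh particles that renders a scrolling, colored plasma energy wave.
# Panning noise distorts shared UVs for soft-noise and plasma samples, feeding opacity and emissive respectively.
# A curve-atlas gradient, offset by V, distortion and the AnimOffset dynamic parameter, sweeps a mask across both.
# Particle color tints the optionally desaturated plasma, while opacity adds V-edge falloff, particle alpha, view-angle, depth and optional near-camera fades.
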Material /Game/Effects/Fort_Effects/Materials/Energy/M_Chromatic_Energy_Wave
graph {
 properties {
  bUsedWithNiagaraMeshParticles=true
  blendMode=BLEND_Translucent
  pixelDepthOffsetMode=PDOM_Legacy
  shadingModel=MSM_Unlit
  twoSided=true
  usedShadingModels=Unlit
 }
 nodes [
  n1 = ParticleColor()
  n2 = TextureCoordinate()
  n3 = AppendVector(A=ScalarParameter() {defaultValue=2, parameterName=BaseNoiseU_Tile}, B=ScalarParameter() {defaultValue=0.5, parameterName=BaseNoiseV_Tile})
  n4 = ScalarParameter() {defaultValue=1.5, parameterName=DistortionU_Tile}
  n5 = ScalarParameter() {defaultValue=0.4, parameterName=DistortionV_Tile}
  n6 = Panner(Coordinate=Multiply(A=TextureCoordinate(), B=AppendVector(A=@n4, B=@n5))) {speedY=-0.3}
  n7 = Divide(
   A=ComponentMask(None=TextureSample(UVs=@n6) {samplerType=SAMPLERTYPE_LinearColor, texture=&T_NoisePack_LinearColor}.RGB) {g=true, r=true}
   B=ScalarParameter() {defaultValue=8, parameterName=DistortionAmount}
  )
  n8 = Panner(
   Coordinate=Add(A=Multiply(A=@n2, B=@n3), B=@n7)
   Speed=AppendVector(A=Constant(), B=ScalarParameter() {defaultValue=0.5, parameterName=BaseNoisePanSpeed})
  ) {speedY=0.5}
  n9 = TextureSample(UVs=@n8) {texture=/Game/Effects/Fort_Effects/Emotes/UltraLight/Texture/T_SoftNoise_BC7}
  n10 = MaterialFunctionCall(Float2=TextureCoordinate()) {materialFunction=&BreakOutFloat2Components}
  n11 = Multiply(B=@n10.G) {constA=1}
  n12 = Multiply(A=OneMinus(None=@n11)) {constB=3}
  n13 = Multiply(A=@n1.A, B=Multiply(A=Multiply(A=@n9.R, B=@n11), B=Saturate(None=@n12)))
  n14 = TextureSample(UVs=@n8) {samplerType=SAMPLERTYPE_LinearColor, texture=/Game/Effects/Fort_Effects/Emotes/UltraLight/Texture/T_Plasma_A_C}
  n15 = Add(
   A=Divide(A=@n14.R, B=ScalarParameter() {defaultValue=8, parameterName=MaskDistortionAmount})
   B=Multiply(A=ComponentMask(None=@n7) {r=true}) {constB=2}
  )
  n16 = DynamicParameter() {paramNames [AnimOffset,Param2,Param3,Param4]}
  n17 = CurveAtlasRowParameter(CurveTime=Add(A=Add(A=@n15, B=@n10.G), B=LinearInterpolate(Alpha=@n16.AnimOffset) {constA=-0.85, constB=0.7}) {constB=0}) {
   atlas=&/CA_FX_Energy
   curve=&/Curve_FX_Energy_01
   parameterName=GradientCurve
  }
  n18 = ChannelMaskParameter(None=MaterialFunctionCall(R=@n17.R, G=@n17.R, B=@n17.R, A=@n17.R) {materialFunction=&AppendMany}.RGBA) {maskChannel=Alpha, parameterName=GradientChannelMask, defaultValue=<rgba 0 0 0 1>}
  n19 = ScalarParameter() {parameterName=FadeBasedOnViewAngle}
  n20 = Multiply(
   A=Multiply(
    A=Multiply(A=@n13, B=@n18.R)
    B=MaterialFunctionCall("FadeSensitivity (optional)"=@n19) {materialFunction=&FadeBasedOnViewAngle}.Result
   )
   B=DepthFade(FadeDistance=ScalarParameter() {parameterName=DepthFadeDistance}) {fadeDistanceDefault=0}
  )
  n21 = MaterialFunctionCall(Distance=ScalarParameter() {group=Opacity, parameterName=NearCameraFadeDistance, sortPriority=5}) {materialFunction=&Near_Camera_Fade}
  n22 = MaterialFunctionCall(
   Fade Length=ScalarParameter() {group=Opacity, parameterName="Cam Fade Length (S)", sortPriority=7}
   Fade Offset=ScalarParameter() {group=Opacity, parameterName="Cam Fade Offset (S)", sortPriority=8}
  ) {materialFunction=&CameraDepthFade}
  n23 = ScalarParameter() {group=Opacity, parameterName=Use Better Camera Fade, sortPriority=6}
  LinearInterpolate(
   A=@n20
   B=Multiply(A=@n20, B=LinearInterpolate(A=@n21.Result, B=@n22.Result, Alpha=@n23))
   Alpha=ScalarParameter() {group=Opacity, parameterName=Use CameraNearFade, sortPriority=4}
  )
  n24 = Desaturation(None=@n14.RGB, Fraction=ScalarParameter() {parameterName=Desatuation})
  MaterialFunctionCall(Emissive=Multiply(A=Multiply(A=@n1.RGB, B=@n24), B=@n18.R)) {materialFunction=&MF_FX_EmissiveAdjustments}
 ]
}
