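# Translucent unlit effect material for an animated electric energy bubble, usable on mesh particles, Niagara meshes and ribbons.
# A scrolling electricity noise mask drives HDR tinted emissive, scaled by particle color, and an adjustable vertex-normal offset.
# Opacity combines the optionally power-tightened noise, a V-axis edge fade, non-mobile depth fade and particle alpha.
# Parameters blend in a radial UV mask and inverted Fresnel, and a static switch adds sine-wave pulsing.
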Material /Game/Effects/Fort_Effects/Materials/Energy/M_ElectricBubble
graph {
 properties {
  bUsedWithMeshParticles=true
  bUsedWithNiagaraMeshParticles=true
  bUsedWithNiagaraRibbons=true
  blendMode=BLEND_Translucent
  pixelDepthOffsetMode=PDOM_Legacy
  shadingModel=MSM_Unlit
  usedShadingModels=Unlit
 }
 nodes [
  n1 = Multiply(A=TextureCoordinate(), B=ScalarParameter() {defaultValue=1, parameterName=CoordMultiplier})
  n2 = ComponentMask(None=@n1) {r=true}
  n3 = Frac(None=Multiply(A=Time(), B=ScalarParameter() {parameterName=Speed}))
  n4 = ComponentMask(None=@n1) {g=true}
  n5 = Frac(None=Multiply(A=Time(), B=ScalarParameter() {defaultValue=0.3, parameterName=SpeedY}))
  n6 = TextureSampleParameter2D(UVs=AppendVector(A=Add(A=@n2, B=@n3), B=Add(A=@n4, B=@n5))) {
   parameterName=Texture
   samplerType=SAMPLERTYPE_Masks
   texture=/Game/Characters/Player/Male/Medium/Bodies/M_MED_Storm_Soldier/Textures/ElectricityNoise_M
  }
  n7 = Power(Base=@n6.G, Exp=ScalarParameter() {defaultValue=4, group=AlphaTighten, parameterName=LightningExponent})
  Clamp(None=@n7)
  n8 = ParticleColor()
  n9 = ComponentMask(None=TextureCoordinate()) {g=true}
  n10 = Multiply(
   A=LinearInterpolate(A=@n6.G, B=@n7, Alpha=ScalarParameter() {group=AlphaTighten, parameterName=LightningTightener?, sliderMax=1})
   B=Multiply(A=@n9, B=OneMinus(None=@n9))
  )
  n11 = ShadingPathSwitch(Default=DepthFade(Opacity=@n10, FadeDistance=ScalarParameter() {defaultValue=350, parameterName=FadeDistance}), Mobile=@n10) {desc=Migrated from FeatureLevelSwitch}
  n12 = ScalarParameter() {defaultValue=0.17, parameterName=HardnessFalloff}
  n13 = Constant() {r=1}
  n14 = SphereMask(
   A=TextureCoordinate()
   B=Constant2Vector() {g=0.5, r=0.5}
   Radius=ScalarParameter() {defaultValue=0.35, group=SphereMask, parameterName=Radius}
   Hardness=ScalarParameter() {group=SphereMask, parameterName=Hardness}
  )
  n15 = ScalarParameter() {defaultValue=1, parameterName=SphereMask?, sliderMax=1}
  n16 = Multiply(A=@n8.A, B=Multiply(A=Multiply(A=@n11, B=@n12), B=LinearInterpolate(A=@n13, B=@n14, Alpha=@n15)))
  n17 = Time()
  n18 = ScalarParameter() {defaultValue=2, parameterName=TimeDilation}
  n19 = Add(A=Sine(None=Divide(A=@n17, B=@n18)), B=ScalarParameter() {defaultValue=0.8, parameterName=SineWaveDisplacement}) {constB=0.8}
  n20 = StaticSwitchParameter(True=Multiply(A=@n16, B=Clamp(None=@n19)), False=@n16) {parameterName=UseTimeForOpacity}
  n21 = ScalarParameter() {defaultValue=1, group=Fresnel, parameterName=FresnelExp}
  LinearInterpolate(
   A=@n20
   B=Multiply(
    A=@n20
    B=MaterialFunctionCall(Exponent=@n21) {materialFunction=&MF_Fresnel_FX}.Fresnel Inverted
   )
   Alpha=ScalarParameter() {group=Fresnel, parameterName=Fresnel?, sliderMax=1}
  )
  n22 = VectorParameter() {parameterName=1231asdf, defaultValue=<rgba 17.3681 3.125 25 1>}
  MaterialFunctionCall(Emissive=Multiply(A=@n8.RGB, B=LinearInterpolate(B=@n22.RGB, Alpha=@n6.G))) {materialFunction=&MF_FX_EmissiveAdjustments}
  Multiply()
  Multiply(A=Multiply(A=@n6.G, B=VertexNormalWS()), B=ScalarParameter() {parameterName=Vertex Normal Amnt})
 ]
}
