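# Additive, unlit, two-sided Niagara sprite and mesh-particle material for flickering electric lightning.
# A random flipbook frame is sampled twice with opposing UV offsets and separate tints, giving a color-split edge.
# A wave gradient sampled through a radial mask around the object pivot creates a ripple that distorts UVs.
# Dynamic parameters drive flicker, erosion, tiling and ripple, with multiplied sines producing the flicker and an optional top/bottom fade.
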Material /Game/Effects/Fort_Effects/Materials/Energy/M_Electric_Lightning_Sprite_Ripple
graph {
 properties {
  bUsedWithNiagaraMeshParticles=true
  bUsedWithNiagaraSprites=true
  blendMode=BLEND_Additive
  pixelDepthOffsetMode=PDOM_Legacy
  shadingModel=MSM_Unlit
  twoSided=true
  usedShadingModels=Unlit
 }
 nodes [
  n1 = ParticleColor()
  n2 = TextureCoordinate()
  n3 = Constant() {r=0.5}
  n4 = SphereMask(A=Power(Base=ComponentMask(None=@n2) {g=true}), B=@n3, Radius=@n3) {hardnessPercent=70}
  n5 = ParticleRandom()
  n6 = DynamicParameter() {defaultValue=<rgba 1 1 0 1>, paramNames [Random,FlickerSpeed,Erode,VerticalTile]}
  n7 = Multiply(A=Add(A=Time(), B=Add(A=@n5, B=@n6.Random)), B=@n6.FlickerSpeed)
  n8 = Sine(None=@n7) {period=0.11}
  n9 = Multiply(A=Sine(None=@n7) {period=0.09}, B=Sine(None=@n7) {period=0.075})
  n10 = MaterialFunctionCall(In=ConstantBiasScale(None=Multiply(A=@n8, B=@n9)), Contrast=ScalarParameter() {defaultValue=3, parameterName=FlickerContrast}) {materialFunction=&CheapContrast}
  n11 = Multiply(A=Max(A=Saturate(None=@n10.Result), B=NamedRerouteUsage()), B=Clamp(None=@n1.A) {clampMode=CMODE_ClampMax})
  n12 = AppendVector(A=Constant() {r=1}, B=@n6.VerticalTile)
  n13 = ScalarParameter() {defaultValue=4, parameterName=NumColumns}
  n14 = MaterialFunctionCall("Animation  Phase (0-1)"=ParticleRandom(), Number of Rows=ScalarParameter() {defaultValue=1, parameterName=NumRows}, Number of Columns=@n13) {materialFunction=&FlipBook}
  n15 = Constant()
  n16 = DynamicParameter() {parameterIndex=1, defaultValue=<rgba 0 0 0 300>, paramNames [VerticalPan,RippleMult,RippleTime,RippleRadius]}
  n17 = WorldPosition()
  n18 = MaterialFunctionCall() {materialFunction=&ObjectPivotPoint}
  n19 = TextureSampleParameter2D(UVs=Add(A=SphereMask(A=@n17.XYZ, B=@n18.Object Pivot Location, Radius=@n16.RippleRadius) {hardnessPercent=0}, B=@n16.RippleTime)) {parameterName=WaveMask, samplerType=SAMPLERTYPE_LinearGrayscale, texture=/Game/Effects/Fort_Effects/Textures/Grad/T_Vertical_Gradient_Wave}
  n20 = Multiply(A=NamedRerouteUsage(), B=@n19.G)
  n21 = Power(Base=@n20) {constExponent=3}
  n22 = TextureCoordinate()
  n23 = Multiply(A=@n21, B=Subtract(A=ComponentMask(None=@n22) {r=true}) {constB=0.5})
  n24 = ScalarParameter() {defaultValue=-0.1, parameterName=UVRippleDistortionX}
  n25 = ScalarParameter() {defaultValue=0.2, parameterName=UVRippleDistortionY}
  n26 = Add(
   A=Add(A=Multiply(A=@n12, B=@n14.UVs), B=AppendVector(A=@n15, B=@n16.VerticalPan))
   B=Multiply(A=AppendVector(A=@n23, B=@n20), B=AppendVector(A=@n24, B=@n25))
  )
  n27 = Divide(A=AppendVector(A=ScalarParameter() {defaultValue=0.007, parameterName=MinOffset}, B=Constant()), B=@n13)
  n28 = LinearInterpolate(A=@n27, B=Multiply(A=@n27) {constB=5}, Alpha=NamedRerouteUsage())
  n29 = LinearInterpolate(Alpha=NamedRerouteUsage()) {constA=1, constB=16}
  n30 = Power(
   Base=TextureSampleParameter2D(UVs=Add(A=@n26, B=@n28) {constB=0.05}) {parameterName=Texture, texture=/Game/Effects/Fort_Effects/Textures/Electric/T_Lightning_Sprites}.G
   Exp=@n29
  )
  n31 = VectorParameter() {parameterName=Color1, defaultValue=<rgba 0.5 0 0.3994 1>}
  n32 = Power(
   Base=TextureSampleParameter2D(UVs=Subtract(A=@n26, B=@n28)) {parameterName=Texture, texture=&T_Lightning_Sprites}.G
   Exp=@n29
  )
  n33 = VectorParameter() {parameterName=Color2, defaultValue=<rgba 0 0.3707 0.4 0>}
  n34 = Multiply(
   A=StaticSwitchParameter(True=Multiply(A=@n4, B=@n11), False=@n11) {defaultValue=true, parameterName=UseTopBottomFade}
   B=Add(A=Multiply(A=@n30, B=@n31.RGB), B=Multiply(A=@n32, B=@n33.RGB))
  )
  n35 = NamedRerouteUsage()
  n36 = LinearInterpolate(
   B=LinearInterpolate(B=VectorParameter() {parameterName=RippleColor, defaultValue=<rgba 1.2 0.4172 0.9197 0>}.RGB, Alpha=NamedRerouteUsage()) {constA=0.3, constB=2.5}
   Alpha=Power(Base=Saturate(None=@n35)) {constExponent=0.6}
  ) {constA=1}
  n37 = Multiply(A=@n1.RGB, B=Desaturation(None=Multiply(A=@n34, B=@n36), Fraction=Constant() {r=-0.5}))
  n38 = Clamp(None=@n37)
  n39 = ScalarParameter() {defaultValue=2, parameterName=HDRMult}
  n40 = Constant() {r=0.1}
  MaterialFunctionCall(Emissive=ShadingPathSwitch(
    Default=MaterialFunctionCall(LDR Input=@n38, HDR Tint=@n37, HDR Intensity Multiplier=@n39, Desaturate Luminance=@n40) {materialFunction=&DeriveHDRfromLDR}.RemappedRange
    Mobile=@n37
   )) {materialFunction=&MF_FX_EmissiveAdjustments}
  NamedRerouteDeclaration(None=@n21) {name=RippleMask, nodeColor=<rgba 0 0.7422 1 1>}
  NamedRerouteDeclaration(None=@n6.Erode) {name=ErodeParam, nodeColor=<rgba 0 0.5312 1 1>}
  NamedRerouteDeclaration(None=@n16.RippleMult) {name=RippleMultParam, nodeColor=<rgba 0.3125 0 1 1>}
  n41 = SphereMask(A=PixelDepth().R, B=Constant(), Radius=Subtract(A=ScalarParameter() {defaultValue=1024, parameterName=FadeNearDistance}) {constB=350}) {hardnessPercent=10}
  n42 = DepthFade(FadeDistance=ScalarParameter() {parameterName=Depth Fade})
  StaticSwitchParameter(True=Multiply(A=LinearInterpolate(Alpha=@n41) {constA=1, constB=0}, B=@n42), False=@n42) {parameterName=UseNearFade}
 ]
}
