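# Translucent unlit particle material for sprites and Niagara ribbons, drawing glowing, wispy ghost smoke.
# A panning spooky-faces texture distorts the UVs of an inky smoke flipbook, with per-particle randomness varying phase and scroll.
# A radial mask, flipbook and faces texture combine through a variable power into one value driving color blend and opacity.
# Particle color tints it; near-camera fade and mask-varied depth fade soften opacity, with depth fade skipped on mobile.
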Material /Game/Effects/Fort_Effects/Materials/Energy/M_Ghost_Wispy_Inky_Glowy_Stuff
graph {
 properties {
  bUsedWithNiagaraRibbons=true
  bUsedWithNiagaraSprites=true
  bUsedWithParticleSprites=true
  blendMode=BLEND_Translucent
  pixelDepthOffsetMode=PDOM_Legacy
  shadingModel=MSM_Unlit
  usedShadingModels=Unlit
 }
 nodes [
  n1 = VectorParameter() {parameterName=DarkColor, defaultValue=<rgba 0.1693 0.3244 0.5 0>}
  n2 = ScalarParameter() {defaultValue=3, parameterName=ColorMult}
  n3 = SphereMask(A=TextureCoordinate(), B=Constant() {r=0.5}, Radius=ScalarParameter() {defaultValue=0.5, parameterName=MaskRadius, sliderMax=1}) {hardnessPercent=0}
  n4 = DynamicParameter() {paramNames [Random,PanAdd,DistortMult,Param4]}
  n5 = ScalarParameter() {defaultValue=0.15, parameterName=Distort}
  n6 = Multiply(A=TextureCoordinate() {uTiling=1.2, vTiling=0.6}, B=ScalarParameter() {defaultValue=1, parameterName=SecondaryTile})
  n7 = Add(A=Add(A=@n4.Random, B=@n4.Random), B=Multiply(A=Time(), B=ScalarParameter() {defaultValue=0.2, parameterName=PanSpeed}))
  n8 = AppendVector(A=Multiply(A=LinearInterpolate(Alpha=@n4.Random) {constA=-1}) {constB=0.35}, B=Constant() {r=1})
  n9 = Desaturation(
   None=TextureSample(UVs=Panner(Coordinate=@n6, Time=@n7, Speed=@n8) {speedX=0.05, speedY=-0.17}) {texture=/Game/Effects/Fort_Effects/Textures/Tile/T_SpookyFaces_Chroma_Tile}.RGB
   Fraction=Constant() {r=-0.5}
  )
  n10 = Add(A=TextureCoordinate(), B=Multiply(A=Multiply(A=@n4.DistortMult, B=@n5), B=ComponentMask(None=@n9) {g=true, r=true}) {constB=-0.1})
  n11 = AppendVector(A=ScalarParameter() {parameterName=Pri_PanSpeedU}, B=ScalarParameter() {parameterName=Pri_PanSpeedV})
  n12 = ComponentMask(None=MaterialFunctionCall(
    "Animation  Phase (0-1)"=@n4.Random
    Texture=TextureObject() {texture=/Game/Effects/Fort_Effects/Textures/T_InkySmoke_Flipbook}
    UVs=Panner(Coordinate=@n10, Time=@n7, Speed=@n11)
   ) {materialFunction=&FlipBook}.Result) {g=true}
  n13 = Add(A=@n3, B=@n12)
  n14 = Multiply(A=@n13) {constB=0.1}
  n15 = LinearInterpolate(A=@n9, B=Multiply(A=@n9, B=@n3), Alpha=ScalarParameter() {parameterName=MaskWeight, sliderMax=1}) {constAlpha=1}
  n16 = Power(
   Base=Multiply(A=@n13, B=Add(A=@n14, B=@n15))
   Exp=LinearInterpolate(A=ScalarParameter() {defaultValue=2, parameterName=PowerMin}, B=ScalarParameter() {defaultValue=0.7, parameterName=PowerMax}, Alpha=@n12) {constA=3, constB=0.7}
  )
  n17 = Clamp(None=@n16)
  n18 = ParticleColor()
  MaterialFunctionCall(Emissive=Multiply(A=LinearInterpolate(A=@n1.RGB, B=@n2, Alpha=@n17), B=@n18.RGB)) {materialFunction=&MF_FX_EmissiveAdjustments}
  n19 = MaterialFunctionCall(Distance=ScalarParameter() {defaultValue=150, parameterName=NearCamFadeDist}) {materialFunction=&Near_Camera_Fade}
  n20 = ScalarParameter() {defaultValue=32, parameterName=DepthFadeMax}
  n21 = ScalarParameter() {defaultValue=6, parameterName=DepthFadeMin}
  n22 = ShadingPathSwitch(Default=DepthFade(FadeDistance=LinearInterpolate(A=@n20, B=@n21, Alpha=@n3)), Mobile=Constant() {r=1}) {desc=Migrated from FeatureLevelSwitch}
  Multiply(A=@n18.A, B=Multiply(A=@n16, B=Multiply(A=@n19.Result, B=@n22)))
 ]
}
