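# Translucent, unlit, two-sided mesh-particle material for the Matter Collector shock wave effect, with refraction and a glow where it intersects scene geometry.
# Two panning cartoon-flame samples, shaped by dynamic parameters and Fresnel, blend via depth fade with flat particle color at intersections.
# Camera-distance fades scale brightness and edge thickness, and time-of-day colors tint the emissive.
# Vertex alpha and near-camera fade drive opacity and, inverted with Fresnel, refraction.
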Material /Game/Effects/Fort_Effects/Materials/Energy/M_MatterCollector_ShockWave
graph {
 properties {
  bEnableResponsiveAA=true
  bUsedWithMeshParticles=true
  blendMode=BLEND_Translucent
  pixelDepthOffsetMode=PDOM_Legacy
  refractionMethod=RM_IndexOfRefraction
  shadingModel=MSM_Unlit
  translucencyPass=MTP_BeforeDOF
  twoSided=true
  usedShadingModels=Unlit
 }
 nodes [
  n1 = VertexColor()
  n2 = DynamicParameter() {defaultValue=<rgba 0 0 0 1>, paramNames [TextureTile, BrightenNoise, Param3, RefractStrength]}
  n3 = Fresnel(ExponentIn=@n2.RefractStrength, Normal=PixelNormalWS())
  n4 = Multiply(A=OneMinus(None=@n1.A), B=@n3)
  n5 = MaterialFunctionCall(Distance=ScalarParameter() {defaultValue=128, parameterName=NearFade}) {materialFunction=&Near_Camera_Fade}
  n6 = OneMinus(None=@n5.Result)
  Clamp(None=LinearInterpolate(Alpha=Multiply(A=@n4, B=@n6)) {constA=0.99})
  n7 = MaterialFunctionCall(Distance=ScalarParameter() {defaultValue=1000, parameterName=LerpColorDistance}) {materialFunction=&Near_Camera_Fade}
  n8 = ScalarParameter() {defaultValue=1, parameterName=ScaleBrightness}
  n9 = ParticleColor()
  n10 = Panner(Coordinate=TextureCoordinate() {uTiling=4, vTiling=4}) {speedY=0.125}
  n11 = Add(A=@n2.BrightenNoise)
  n12 = TextureCoordinate()
  n13 = ParticleRelativeTime()
  n14 = TextureSampleParameter2D(UVs=Panner(Coordinate=Multiply(A=@n12, B=@n2.TextureTile), Time=Multiply(A=@n13, B=@n2.Param3)) {speedY=0.5}) {parameterName=NoiseTexture, texture=/Game/Effects/Fort_Effects/Textures/Tile/T_Tile_Texture_Cartoon_Flames}
  n15 = Multiply(
   A=Multiply(
    A=TextureSampleParameter2D(UVs=@n10) {parameterName=MaskTexture, texture=&T_Tile_Texture_Cartoon_Flames}.RGB
    B=LinearInterpolate(B=@n11, Alpha=@n14.R) {constB=4}
   )
   B=@n3
  )
  n16 = LinearInterpolate(
   A=ScalarParameter() {defaultValue=25, parameterName=EdgeThicknessNear}
   B=ScalarParameter() {defaultValue=50, parameterName=EdgeThicknessFar}
   Alpha=@n7.Result
  )
  n17 = Multiply(
   A=LinearInterpolate(
    A=ScalarParameter() {defaultValue=1, parameterName=NearColorScalar}
    B=ScalarParameter() {defaultValue=4, parameterName=FarColorScalar}
    Alpha=@n7.Result
   )
   B=LinearInterpolate(A=Multiply(A=@n8, B=@n9.RGB), B=Multiply(A=@n9.RGB, B=@n15), Alpha=DepthFade(FadeDistance=@n16) {fadeDistanceDefault=75})
  )
  n18 = Multiply(A=Multiply(A=@n17, B=@n1.A), B=@n9.A)
  n19 = LinearInterpolate(
   A=ScalarParameter() {defaultValue=15, parameterName=EdgeThicknessNear}
   B=ScalarParameter() {defaultValue=30, parameterName=EdgeThicknessFar}
  )
  Clamp(None=Multiply(A=DepthFade(Opacity=@n18, FadeDistance=@n19) {fadeDistanceDefault=75}, B=@n5.Result))
  n20 = VectorParameter() {parameterName=ColorMorning, defaultValue=<rgba 1 1 1 0>}
  n21 = VectorParameter() {parameterName=ColorDay, defaultValue=<rgba 1 1 1 0>}
  n22 = VectorParameter() {parameterName=Color_Evening, defaultValue=<rgba 1 1 1 0>}
  n23 = VectorParameter() {parameterName=ColorNight, defaultValue=<rgba 1 1 1 0>}
  MaterialFunctionCall(Emissive=Multiply(
    A=MaterialFunctionCall(Color_Morning=@n20.RGB, Color_Day=@n21.RGB, Color_Evening=@n22.RGB, Color_Night=@n23.RGB) {materialFunction=&Color_Time_of_Day}.Result
    B=@n17
   )) {materialFunction=&MF_FX_EmissiveAdjustments}
 ]
}
