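# Translucent, unlit, two-sided mesh-particle material for a soft, animated plasma cone effect.
# Four panning samples of a chromatic inky smoke texture are averaged, saturation-boosted via negative desaturation and tinted by particle color for emissive.
# Opacity multiplies that noise by particle alpha and a SphereMask on U-collapsed UVs, positioned by the PannerX dynamic parameter.
# Near-camera and depth fades soften opacity, except on the mobile shading path.
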Material /Game/Effects/Fort_Effects/Materials/Energy/M_Plasma_Cone_Soft
graph {
 properties {
  bUsedWithMeshParticles=true
  blendMode=BLEND_Translucent
  pixelDepthOffsetMode=PDOM_Legacy
  shadingModel=MSM_Unlit
  twoSided=true
  usedShadingModels=Unlit
 }
 nodes [
  n1 = ParticleColor()
  n2 = Panner(Coordinate=TextureCoordinate() {uTiling=2, vTiling=0.25}) {speedX=-0.5, speedY=-1}
  n3 = Panner(Coordinate=TextureCoordinate() {uTiling=2, vTiling=0.5}) {speedX=1, speedY=-2}
  n4 = Panner(Coordinate=TextureCoordinate() {uTiling=2, vTiling=0.5}) {speedX=-0.5, speedY=-1}
  n5 = Panner(Coordinate=TextureCoordinate() {uTiling=3}) {speedX=1, speedY=-2}
  n6 = LinearInterpolate(
   A=LinearInterpolate(
    A=TextureSample(UVs=@n2) {texture=&T_Inky_Smoke_Tile_Chromatic}.RGB
    B=TextureSample(UVs=@n3) {texture=&T_Inky_Smoke_Tile_Chromatic}.RGB
   )
   B=LinearInterpolate(
    A=TextureSample(UVs=@n4) {texture=&T_Inky_Smoke_Tile_Chromatic}.RGB
    B=TextureSample(UVs=@n5) {texture=&T_Inky_Smoke_Tile_Chromatic}.RGB
   )
  )
  n7 = Constant() {r=-0.5}
  MaterialFunctionCall(Emissive=Multiply(A=@n1.RGB, B=Desaturation(None=@n6, Fraction=@n7)) {constB=1.5}) {materialFunction=&MF_FX_EmissiveAdjustments}
  n8 = ScalarParameter() {defaultValue=250, parameterName=Camera_FadeDistance}
  n9 = TextureCoordinate() {uTiling=0}
  n10 = DynamicParameter() {defaultValue=<rgba 0 0.5 1 1>, paramNames [PannerX,PannerY,Param3,Param4]}
  n11 = AppendVector(A=@n10.PannerX, B=@n10.PannerX)
  n12 = ScalarParameter() {defaultValue=0.25, parameterName=SoftMask_Radius}
  n13 = ScalarParameter() {defaultValue=-1, parameterName=SoftMask_Hardness}
  n14 = Multiply(A=Multiply(A=@n1.A, B=SphereMask(A=@n9, B=@n11, Radius=@n12, Hardness=@n13)), B=@n6)
  n15 = ScalarParameter() {defaultValue=250, parameterName=Depth_FadeDistance}
  ShadingPathSwitch(
   Default=Multiply(
    A=MaterialFunctionCall(Distance=@n8) {materialFunction=&Near_Camera_Fade}.Result
    B=DepthFade(Opacity=@n14, FadeDistance=@n15)
   )
   Mobile=@n14
  ) {desc=Migrated from FeatureLevelSwitch}
 ]
}
