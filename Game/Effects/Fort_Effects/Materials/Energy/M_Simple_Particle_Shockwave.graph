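# Unlit, two-sided translucent material for particle shockwave effects on sprites and meshes, built on a ground-ring texture.
# Emissive switches between particle color, hue-shifted rainbow, or a "mellow" panning gradient with derived HDR.
# Opacity combines the ring mask, particle alpha and dynamic scale, then optional non-mobile depth fade and time-of-day scaling.
# The ring mask and particle alpha also drive refraction strength; ring rotation uses doubled RotationInitial, leaving RotationTime unused.
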Material /Game/Effects/Fort_Effects/Materials/Energy/M_Simple_Particle_Shockwave
graph {
 properties {
  bUsedWithMeshParticles=true
  bUsedWithNiagaraMeshParticles=true
  bUsedWithNiagaraSprites=true
  bUsedWithParticleSprites=true
  blendMode=BLEND_Translucent
  pixelDepthOffsetMode=PDOM_Legacy
  refractionMethod=RM_IndexOfRefraction
  shadingModel=MSM_Unlit
  twoSided=true
  usedShadingModels=Unlit
 }
 nodes [
  n1 = ParticleColor()
  n2 = DynamicParameter() {paramNames [ScaleAlphaOnly, Param2, RotationInitial, RotationTime]}
  n3 = TextureSample(UVs=Rotator(Time=Add(A=@n2.RotationInitial, B=@n2.RotationInitial))) {samplerSource=SSM_Clamp_WorldGroupSettings, texture=/Game/Effects/Fort_Effects/Textures/Outlander/T_InTheZone_GroundRing_01}
  LinearInterpolate(B=ScalarParameter() {defaultValue=1.08, parameterName=RefractionIndex}, Alpha=Multiply(A=@n1.A, B=@n3.G)) {constA=1, constB=1.1}
  n4 = ComponentMask(None=TextureCoordinate() {uTiling=0, vTiling=0}) {g=true}
  n5 = Multiply(
   A=TextureSample(UVs=Panner(Coordinate=@n4) {speedX=1.6667}) {texture=/Game/Global/Textures/Default/Masks_and_Gradients/T_Grad_Rainbow}.RGB
   B=ParticleColor().RGB
  )
  n6 = Clamp(None=@n5)
  n7 = ScalarParameter() {defaultValue=1, parameterName=Speed}
  n8 = Time()
  n9 = Multiply(
   A=MaterialFunctionCall(Hue Shift Percentage=Multiply(A=@n7, B=@n8), Texture=VectorParameter() {parameterName=StartColor, defaultValue=<rgba 1 0 0 1>}.RGB) {materialFunction=&HueShift}.Result
   B=@n1.RGB
  )
  MaterialFunctionCall(Emissive=StaticSwitchParameter(
    True=MaterialFunctionCall(LDR Input=@n6, HDR Tint=@n5) {materialFunction=&DeriveHDRfromLDR}.RemappedRange
    False=StaticSwitchParameter(True=@n9, False=@n1.RGB) {parameterName=UseRainbowColor}
   ) {parameterName=UseMellowRainbowColors}) {materialFunction=&MF_FX_EmissiveAdjustments}
  n10 = Multiply(A=Multiply(A=@n2.ScaleAlphaOnly, B=@n1.A), B=@n3.G)
  n11 = ScalarParameter() {defaultValue=100, parameterName=DepthFadeDistance}
  n12 = StaticSwitchParameter(True=ShadingPathSwitch(Default=DepthFade(Opacity=@n10, FadeDistance=@n11), Mobile=@n10) {desc=Migrated from FeatureLevelSwitch}, False=@n10) {parameterName=UseDepthFade}
  MaterialFunctionCall(
   Value_Morning=Multiply(A=ScalarParameter() {defaultValue=1, parameterName=Morning_opacity_scale}, B=@n12)
   Value_Day=Multiply(A=ScalarParameter() {defaultValue=1, parameterName=Day_opacity_scale}, B=@n12)
   Value_Evening=Multiply(A=ScalarParameter() {defaultValue=1, parameterName=Evening_opacity_scale}, B=@n12)
   Value_Night=Multiply(A=ScalarParameter() {defaultValue=1, parameterName=Night_opacity_scale}, B=@n12)
  ) {materialFunction=&Scalar_Time_of_Day}
  Reroute(None=TextureSample() {samplerType=SAMPLERTYPE_Normal, texture=/Game/Environments/World/Backgrounds/Universal/TechContent/Textures/T_Spherenormal_01}.RGB)
 ]
}
