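# Translucent unlit particle material for soft shockwave effects, with refraction distortion and optional rainbow tint.
# Opacity combines particle alpha, paired Fresnel masks, bottom gradient falloff and depth fade; emissive blends particle color with a Fresnel-indexed curve-atlas gradient.
# Dynamic-parameter refraction index follows the mask, optionally scaled by camera distance and by FOV read from the projection matrix.
# Low quality, forward and mobile paths bypass refraction and the Fresnel/depth-fade opacity.
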Material /Game/Effects/Fort_Effects/Materials/Energy/M_Soft_Shockwave
graph {
 properties {
  bUsedWithMeshParticles=true
  bUsedWithNanite=true
  bUsedWithNiagaraMeshParticles=true
  bUsedWithNiagaraRibbons=true
  bUsedWithNiagaraSprites=true
  blendMode=BLEND_Translucent
  pixelDepthOffsetMode=PDOM_Legacy
  refractionMethod=RM_IndexOfRefraction
  shadingModel=MSM_Unlit
  usedShadingModels=Unlit
 }
 nodes [
  n1 = ParticleColor()
  n2 = DynamicParameter() {paramNames [RefractionIndex, RainbowColor, ExponentIn, BaseReflectFractionIn]}
  n3 = Fresnel(ExponentIn=@n2.ExponentIn, BaseReflectFractionIn=@n2.ExponentIn)
  n4 = VectorParameter() {
   group=RainbowColor
   parameterName=RainbowColor
   channelNames {a="", b=Desaturation, g=FresnelInvert, r=Use Rainbow Color}
   defaultValue=<rgba 0 1 0 0>
  }
  n5 = CurveAtlasRowParameter(CurveTime=LinearInterpolate(A=@n3, B=OneMinus(None=@n3), Alpha=@n4.FresnelInvert)) {
   atlas=&AthenaCurveAtlas01
   curve=/Game/Athena/Environments/Nexus/Materials/FXshader/Curve_Color
   defaultValue=78
   group=RainbowColor
   parameterName=RainbowColor
  }
  n6 = Multiply(
   A=DynamicParameter() {paramNames [RefractionIndex, RainbowColor, ExponentIn, BaseReflectFractionIn]}.RainbowColor
   B=Desaturation(None=@n5, Fraction=@n4.Desaturation)
  )
  MaterialFunctionCall(Emissive=LinearInterpolate(A=@n1.RGB, B=Multiply(A=@n6, B=@n1.RGB), Alpha=@n4.Use Rainbow Color)) {materialFunction=&MF_FX_EmissiveAdjustments}
  n7 = MaterialFunctionCall(
   In=OneMinus(None=MaterialFunctionCall() {materialFunction=&LinearGradient}.VGradient)
   Contrast=ScalarParameter() {defaultValue=0.05, parameterName=BottomFalloff}
  ) {materialFunction=&CheapContrast}
  n8 = ScalarParameter() {defaultValue=1, parameterName=FresnelExponent}
  n9 = ScalarParameter() {defaultValue=1, parameterName=FresnelExponent2}
  n10 = MaterialFunctionCall(In=OneMinus(None=Fresnel(ExponentIn=@n9) {exponent=1}), Contrast=Constant() {r=1}) {materialFunction=&CheapContrast}
  n11 = Multiply(
   A=MaterialFunctionCall(In=Fresnel(ExponentIn=@n8) {exponent=1}, Contrast=Constant() {r=0.1}) {materialFunction=&CheapContrast}.Result
   B=@n10.Result
  )
  n12 = ScalarParameter() {defaultValue=1, parameterName=FresnelBrightness}
  n13 = DepthFade(Opacity=Clamp(None=Multiply(A=@n11, B=@n12)), FadeDistance=ScalarParameter() {parameterName=DepthFade_Distance})
  n14 = Constant() {r=1}
  n15 = QualitySwitch(Default=@n13, Low=@n14, High=@n13, Medium=@n13, Epic=@n13)
  Multiply(A=@n1.A, B=Multiply(A=@n7.Result, B=ShadingPathSwitch(Default=@n15, Deferred=@n15, Forward=@n14, Mobile=@n14)))
  n16 = Custom!(None) {code="atan(View.ClipToView[0][0])*2", desc=FOV in radians, description=GrabFOV, outputType=CMOT_Float2}
  n17 = ScalarParameter() {defaultValue=30, desc="Camera FOV Min Angle (long lens)", group=CameraFOVFade, parameterName=FOV Degrees Min, sliderMax=30, sortPriority=20}
  n18 = ScalarParameter() {defaultValue=80, desc="Camera FOV Max Angle (wide lens)", group=CameraFOVFade, parameterName=FOV Degrees Max, sliderMax=90, sortPriority=21}
  n19 = ScalarParameter() {defaultValue=0.4, desc="IOR at Min (long lens) Angle", group=CameraFOVFade, parameterName=FOV Refract Min, sliderMax=1, sortPriority=22}
  n20 = ScalarParameter() {defaultValue=1, desc="IOR at Max(wide lens) Angle", group=CameraFOVFade, parameterName=FOV Refract Max, sliderMax=1, sortPriority=23}
  n21 = MaterialFunctionCall(
   Input=ComponentMask(None=Multiply(A=@n16) {constB=57.2958, desc=to degrees}) {g=true}
   Input Low=@n17
   Input High=@n18
   Target Low=@n19
   Target High=@n20
  ) {materialFunction=&RemapValueRange}
  OneMinus(None=@n21.Result)
  n22 = Multiply(A=@n7.Result, B=@n10.Result)
  Saturate(None=@n22)
  n23 = StaticSwitchParameter(True=Clamp(None=@n22), False=@n22) {parameterName=Use Clamp}
  n24 = ScalarParameter() {defaultValue=4096, group=CameraFade, parameterName=Refraction CameraFade Distance}
  n25 = Multiply(
   A=@n23
   B=OneMinus(None=MaterialFunctionCall(Distance=@n24) {materialFunction=&Near_Camera_Fade}.Result)
  )
  n26 = ScalarParameter() {group=CameraFade, parameterName=Use Refraction CameraFade, sortPriority=0}
  n27 = ScalarParameter() {desc=Use Camera FOV to Fade, group=CameraFOVFade, parameterName=Use FOV Scale, sortPriority=0}
  n28 = LinearInterpolate(
   B=DynamicParameter() {paramNames [RefractionIndex, RainbowColor, ExponentIn, BaseReflectFractionIn]}.RefractionIndex
   Alpha=Multiply(A=LinearInterpolate(A=@n23, B=@n25, Alpha=@n26) {constA=1, constB=0}, B=LinearInterpolate(B=@n21.Result, Alpha=@n27) {constA=1}) {desc="0=NoRefract"}
  ) {constA=1}
  n29 = Constant() {r=1}
  n30 = QualitySwitch(Default=@n28, Low=@n29, High=@n28, Medium=@n28, Epic=@n28)
  StaticSwitchParameter(True=Constant() {r=1}, False=ShadingPathSwitch(Default=@n30, Deferred=@n30, Forward=@n29, Mobile=@n29)) {parameterName=DisableRefraction}
 ]
}
