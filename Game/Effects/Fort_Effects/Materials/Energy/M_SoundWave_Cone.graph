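# Translucent, unlit, two-sided energy effect material for mesh particles, apparently a sound-wave cone, tinted by particle color with optional animated hue shift.
# Four panned textures with power-stretched V coordinates, partly normal-map distorted, build emissive and opacity, timed by Time or a dynamic parameter.
# Edge, depth, camera and optional Fresnel fades shape opacity.
# Normal-based vertex offset and optional refraction add warping, both disabled on mobile and lower quality levels.
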Material /Game/Effects/Fort_Effects/Materials/Energy/M_SoundWave_Cone
graph {
 properties {
  bUsedWithMeshParticles=true
  bUsedWithNiagaraMeshParticles=true
  blendMode=BLEND_Translucent
  pixelDepthOffsetMode=PDOM_Legacy
  refractionMethod=RM_IndexOfRefraction
  shadingModel=MSM_Unlit
  twoSided=true
  usedShadingModels=Unlit
 }
 nodes [
  n1 = ParticleColor()
  n2 = Multiply(
   A=StaticSwitchParameter(True=Time(), False=DynamicParameter() {paramNames [TimeMutiplier, Param2, Param3, Ref]}.TimeMutiplier) {defaultValue=true, parameterName=Debug}
   B=ScalarParameter() {defaultValue=1, parameterName=TimeMutiplier}
  )
  n3 = StaticSwitchParameter(
   True=MaterialFunctionCall(Hue Shift Percentage=@n2, Texture=Constant3Vector() {constant=<rgba 0 1 1 0>}) {materialFunction=&HueShift}.Result
   False=Constant() {r=1}
  ) {defaultValue=true, parameterName=UseHueShift}
  n4 = MaterialFunctionCall(Float2=TextureCoordinate() {vTiling=1.3}) {materialFunction=&BreakOutFloat2Components}
  n5 = ScalarParameter() {defaultValue=1, parameterName=VTile_PowerStretcher}
  n6 = AppendVector(A=@n4.R, B=Power(Base=@n4.G, Exp=Multiply(B=@n5) {constA=6}) {constExponent=6})
  n7 = AppendVector(A=ScalarParameter() {defaultValue=0.6, parameterName=MasterMaskU}, B=ScalarParameter() {defaultValue=1, parameterName=MasterMaskV})
  n8 = TextureSampleParameter2D(UVs=Multiply(A=Panner(Coordinate=@n6, Time=@n2) {speedX=0.5, speedY=-4}, B=@n7)) {group=TextureSamplers, parameterName=TextureA, texture=/Game/Effects/Fort_Effects/Textures/Energy/T_EnergyFlow_Packed}
  n9 = MaterialFunctionCall(Float2=TextureCoordinate()) {materialFunction=&BreakOutFloat2Components}
  n10 = Multiply(B=@n5) {constA=4}
  n11 = MaterialFunctionCall(Float2=TextureCoordinate() {uTiling=2, vTiling=0.6}) {materialFunction=&BreakOutFloat2Components}
  n12 = Power(Base=@n11.G, Exp=Multiply(B=@n5) {constA=2})
  n13 = TextureSample(UVs=Panner(Coordinate=AppendVector(A=@n11.R, B=@n12), Time=@n2) {speedY=-0.5}) {samplerType=SAMPLERTYPE_Normal, texture=&T_Weap_Distort_N_02}
  n14 = Constant()
  n15 = ShadingPathSwitch(Default=Multiply(A=ComponentMask(None=@n13.RGB) {g=true, r=true}) {constB=0.12}, Mobile=@n14) {desc=Migrated from FeatureLevelSwitch}
  n16 = QualitySwitch(Default=@n15, Low=@n14, High=@n15, Medium=@n14)
  n17 = Add(A=AppendVector(A=@n9.R, B=Power(Base=@n9.G, Exp=@n10) {constExponent=4}), B=@n16)
  n18 = TextureSampleParameter2D(UVs=Multiply(A=Panner(Coordinate=@n17, Time=@n2) {speedX=0.1, speedY=-1}, B=@n7)) {
   group=TextureSamplers
   parameterName=TextureB
   samplerType=SAMPLERTYPE_LinearColor
   texture=/Game/Effects/Fort_Effects/Textures/Tile/T_AMR_Beam_02
  }
  n19 = Constant()
  n20 = ShadingPathSwitch(Default=Add(A=Multiply(A=@n8.RGB) {constB=0.35}, B=Multiply(A=@n18.RGB) {constB=0.4}), Mobile=@n19) {desc=Migrated from FeatureLevelSwitch}
  n21 = Panner(Coordinate=@n17, Time=@n2) {speedX=-0.2, speedY=-1}
  n22 = Add(
   A=QualitySwitch(Default=@n20, Low=@n19, High=@n20, Medium=@n19)
   B=TextureSampleParameter2D(UVs=Multiply(A=@n21, B=@n7)) {group=TextureSamplers, parameterName=TextureC, texture=&T_EnergyFlow_Packed}.RGB
  )
  n23 = MaterialFunctionCall(Float2=TextureCoordinate() {vTiling=0.7}) {materialFunction=&BreakOutFloat2Components}
  n24 = Add(A=@n16, B=AppendVector(A=@n23.R, B=Power(Base=@n23.G) {constExponent=4}))
  n25 = TextureSampleParameter2D(UVs=Multiply(A=Panner(Coordinate=@n24, Time=@n2) {speedX=0.2, speedY=-1}, B=@n7)) {group=TextureSamplers, parameterName=TextureD, texture=&T_EnergyFlow_Packed}
  n26 = TextureCoordinate() {uTiling=2}
  n27 = LinearInterpolate(Alpha=Distance(A=ComponentMask(None=@n26) {r=true}, B=Constant() {r=1})) {constA=10, constB=0}
  n28 = TextureSample() {texture=/Game/Effects/Fort_Effects/Textures/Masks/T_Vert_Gradient_Softer}
  n29 = ScalarParameter() {defaultValue=1, parameterName=CenterMaskEmis}
  n30 = Add(
   A=Multiply(A=Add(A=@n22, B=@n25.RGB), B=Saturate(None=@n27))
   B=StaticSwitchParameter(True=Multiply(A=@n28.R, B=@n29), False=Constant()) {defaultValue=true, parameterName=UseBrightCore}
  )
  n31 = Desaturation(None=@n30, Fraction=ScalarParameter() {defaultValue=0.4, parameterName=Desaturation})
  MaterialFunctionCall(Emissive=Multiply(A=@n1.RGB, B=Multiply(A=@n3, B=@n31))) {materialFunction=&MF_FX_EmissiveAdjustments}
  n32 = ScalarParameter() {defaultValue=1, parameterName=FadeBasedOnViewAngle}
  n33 = MaterialFunctionCall(
   Brightness=ScalarParameter() {defaultValue=1, group=Opacity, parameterName=Fresnel_Brightness}
   Exponent=ScalarParameter() {defaultValue=2, group=Opacity, parameterName=Fresnel_Exponent}
  ) {materialFunction=&MF_Fresnel_FX}
  n34 = Constant() {r=1}
  n35 = ShadingPathSwitch(
   Default=MaterialFunctionCall(Distance=ScalarParameter() {parameterName=CameraFadeDis}) {materialFunction=&Near_Camera_Fade}.Result
   Mobile=@n34
  ) {desc=Migrated from FeatureLevelSwitch}
  n36 = MaterialFunctionCall(Float2=TextureCoordinate()) {materialFunction=&BreakOutFloat2Components}
  n37 = Saturate(None=Multiply(A=OneMinus(None=@n36.G)))
  n38 = TextureCoordinate() {uTiling=0.5, vTiling=0.5}
  n39 = Constant() {r=1}
  n40 = ShadingPathSwitch(
   Default=TextureSample(UVs=Panner(Coordinate=@n38, Time=@n2) {speedX=0.2, speedY=-1}) {samplerType=SAMPLERTYPE_LinearGrayscale, texture=/Game/Global/Textures/Default/Noise/T_SwirlNoise_N_height}.R
   Mobile=@n39
  ) {desc=Migrated from FeatureLevelSwitch}
  n41 = QualitySwitch(Default=@n40, Low=@n39, High=@n40, Medium=@n39)
  n42 = Multiply(A=Multiply(A=@n30, B=Multiply(A=@n37, B=@n41)), B=ScalarParameter() {defaultValue=2, parameterName=OpacityOverall})
  n43 = ShadingPathSwitch(
   Default=DepthFade(Opacity=@n42, FadeDistance=ScalarParameter() {defaultValue=110, parameterName=FadeDistance}) {fadeDistanceDefault=110}
   Mobile=@n42
  ) {desc=Migrated from FeatureLevelSwitch}
  n44 = Fresnel(Normal=Multiply(A=VertexNormalWS()) {constB=-1}) {baseReflectFraction=0, exponent=1}
  n45 = Fresnel() {baseReflectFraction=0, exponent=1}
  n46 = Constant() {r=1}
  n47 = ShadingPathSwitch(Default=LinearInterpolate(A=OneMinus(None=@n44), B=OneMinus(None=@n45), Alpha=TwoSidedSign()), Mobile=@n46) {desc=Migrated from FeatureLevelSwitch}
  n48 = Multiply(
   A=Multiply(A=@n1.A, B=QualitySwitch(Default=@n43, Low=@n42))
   B=StaticSwitchParameter(True=QualitySwitch(Default=@n47, Low=@n46, Medium=@n46), False=Constant() {r=1}) {defaultValue=true, parameterName=UseEdgeFade}
  )
  n49 = Subtract(A=@n36.G, B=ScalarParameter() {parameterName=GradientMaskBias})
  n50 = Multiply(A=@n48, B=Power(Base=Saturate(None=@n49)) {constExponent=1.5})
  n51 = Power(
   Base=Multiply(
    A=QualitySwitch(Default=@n35, Low=@n34, High=@n35, Medium=@n35)
    B=StaticSwitchParameter(True=@n48, False=@n50) {defaultValue=true, parameterName=UseBrightCore}
   )
   Exp=ScalarParameter() {defaultValue=1, parameterName=Opacity_Exponent}
  )
  Multiply(
   A=LinearInterpolate(
    B=MaterialFunctionCall("FadeSensitivity (optional)"=@n32) {materialFunction=&FadeBasedOnViewAngle}.Result
    Alpha=ScalarParameter() {parameterName=UseFadeBasedOnViewAngle}
   ) {constA=1}
   B=StaticSwitchParameter(True=Multiply(A=@n51, B=@n33.Fresnel Double Sided), False=@n51) {group=Opacity, parameterName=UseFresnelOpacity}
  )
  n52 = VertexNormalWS()
  n53 = Multiply(A=@n13.RGB, B=@n36.G)
  n54 = Constant()
  n55 = ShadingPathSwitch(Default=Multiply(A=Multiply(A=@n52, B=@n53), B=ScalarParameter() {defaultValue=45, parameterName=WPO_Overall}) {constB=45}, Mobile=@n54) {desc=Migrated from FeatureLevelSwitch}
  Reroute(None=QualitySwitch(Default=@n55, Low=@n54, High=@n55, Medium=@n54))
  n56 = TextureCoordinate() {uTiling=2}
  n57 = MaterialFunctionCall(Float2=@n56) {materialFunction=&BreakOutFloat2Components}
  n58 = Panner(Coordinate=AppendVector(A=@n57.R, B=Power(Base=@n57.G) {constExponent=4}), Time=@n2) {speedX=0.2, speedY=-1}
  n59 = Multiply(A=MaterialFunctionCall(Float2=@n56) {materialFunction=&BreakOutFloat2Components}.G) {constB=0.5}
  n60 = Multiply(
   A=OneMinus(None=Fresnel() {baseReflectFraction=0, exponent=1})
   B=Multiply(A=TextureSample(UVs=@n58) {texture=/Game/Effects/Fort_Effects/Textures/Tile/T_LaserTrail}.R, B=Sine(None=@n59))
  )
  n61 = DynamicParameter() {paramNames [TimeMutiplier, Param2, Param3, Ref]}
  n62 = Constant() {r=1}
  n63 = ShadingPathSwitch(Default=LinearInterpolate(Alpha=Multiply(A=@n60, B=@n61.Ref)) {constA=1, constB=1.03}, Mobile=@n62) {desc=Migrated from FeatureLevelSwitch}
  Reroute(None=StaticSwitchParameter(True=QualitySwitch(Default=@n63, Low=@n62, High=@n63, Medium=@n62), False=Constant() {r=1}) {defaultValue=true, parameterName=UseRefraction})
  n64 = Constant3Vector() {constant=<rgba 0 0 1 0>}
  n65 = ShadingPathSwitch(Default=@n13.RGB, Mobile=@n64) {desc=Migrated from FeatureLevelSwitch}
  Reroute(None=StaticSwitchParameter(True=QualitySwitch(Default=@n65, Low=@n64, High=@n65, Medium=@n64), False=@n64) {defaultValue=true, parameterName=UseRefraction})
 ]
}
